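# Physics asset for the female medium Cardboard Crew body, covering secondary motion of its cardboard torso pieces and shoelaces.
# Box bodies for four torso flaps and two side boxes hang from base capsules linked to spine_03; flaps hinge on one swing axis, while boxes rotate freely.
# Each shoe has two two-segment capsule lace chains with limited swing and tight twist.
# Pelvis, spine and foot-ball capsules supply body shapes.
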PhysicsAsset /BRCosmetics/Characters/Player/Female/Medium/Bodies/F_MED_Cardboard_Crew/Meshes/F_MED_Cardboard_Crew_Physics
bodies {
 ball_l [{
   shapeType=Capsule
   length=10.2182
   radius=3.7913
   shapeName=ball_l_capsule
   center=<-5.2868 -0.2844 -0.0003>
   rotation=<rot 90 10.6197 -23.5711>
  }]
 ball_r [{shapeType=Capsule, length=10.2182, radius=3.7913, shapeName=ball_l_capsule, center=<5.2868 0.2844 0.0003>, rotation=<rot 90 0 145.8092>}]
 dyn_laceIn01_I_MSTR [{shapeType=Capsule, length=1, radius=1, shapeName=dyn_laceIn01_I_MSTR_capsule}]
 dyn_laceIn01_l [{shapeType=Capsule, length=1.6818, radius=0.8, shapeName=dyn_laceOut01_l_capsule, center=<1.8305 0 0.0001>, rotation=<rot 0 90 90>}]
 dyn_laceIn01_r [{shapeType=Capsule, length=1.6818, radius=0.8, shapeName=dyn_laceOut01_l_capsule, center=<1.4134 0 -0.0001>, rotation=<rot 0 90 -90>}]
 dyn_laceIn01_r_MSTR [{shapeType=Capsule, length=1, radius=1, shapeName=dyn_laceIn01_r_MSTR_capsule, rotation=<rot 0.0001 0 89.9996>}]
 dyn_laceIn02_l [{shapeType=Capsule, length=4.058, radius=0.8, shapeName=dyn_laceOut02_l_capsule, center=<1.7156 0 0>, rotation=<rot -90 -179.9994 179.9993>}]
 dyn_laceIn02_r [{shapeType=Capsule, length=4.058, radius=0.8, shapeName=dyn_laceOut02_l_capsule, center=<1.7156 0 0>, rotation=<rot -90 180 179.9999>}]
 dyn_laceOut01_l [{shapeType=Capsule, length=1.6818, radius=0.8, shapeName=dyn_laceOut01_l_capsule, center=<1.8305 0 0.0001>, rotation=<rot 0 90 90>}]
 dyn_laceOut01_l_MSTR [{shapeType=Capsule, length=1, radius=1, shapeName=dyn_laceOut01_l_MSTR_capsule}]
 dyn_laceOut01_r [{shapeType=Capsule, length=1.6818, radius=0.8, shapeName=dyn_laceOut01_l_capsule, center=<1.1235 0 -0.0001>, rotation=<rot 0 90 -90>}]
 dyn_laceOut01_r_MSTR [{shapeType=Capsule, length=1, radius=1, shapeName=dyn_lace_l_Base_capsule, rotation=<rot -0.0002 -0.0001 -89.9997>}]
 dyn_laceOut02_l [{shapeType=Capsule, length=4.058, radius=0.8, shapeName=dyn_laceOut02_l_capsule, center=<1.7156 0 0>, rotation=<rot -90 180 179.9999>}]
 dyn_laceOut02_r [{shapeType=Capsule, length=4.058, radius=0.8, shapeName=dyn_laceOut02_l_capsule, center=<1.7156 0 0>, rotation=<rot -90 -179.5252 179.5252>}]
 dyn_shoelace_l_Base [{shapeType=Capsule, length=15, radius=3.7427, shapeName=dyn_shoelace_l_Base_capsule, rotation=<rot 90 46.7471 46.7472>}]
 dyn_shoelace_r_Base [{shapeType=Capsule, length=15, radius=3.0907, shapeName=dyn_shoelace_r_Base_capsule, rotation=<rot 90 46.7471 46.7475>}]
 dyn_torsoBox_Base_l [{shapeType=Capsule, length=0.505, radius=0.505, shapeName=dyn_torsoBox_Base_l_capsule, rotation=<rot 0 -90 0>}]
 dyn_torsoBox_Base_r [{shapeType=Capsule, length=0.505, radius=0.505, shapeName=dyn_torsoBox_Base_r_capsule, rotation=<rot 0 -90 0>}]
 dyn_torsoBox_l [{shapeType=Box, extentX=9.944, extentY=10.8755, extentZ=9.4239, center=<1.9555 -1.8013 -0.287>, rotation=<rot 45.0009 0 0>}]
 dyn_torsoBox_r [{shapeType=Box, extentX=8.4095, extentY=11.2993, extentZ=9.3389, center=<2.387 -1.4648 0.4346>, rotation=<rot 44.9319 3.9746 2.8093>}]
 dyn_torsoFlap_Base_l [{shapeType=Capsule, length=0.505, radius=0.505, shapeName=dyn_torsoFlap_Base_l_capsule, rotation=<rot 0 -90 0>}]
 dyn_torsoFlap_Base_r [{shapeType=Capsule, length=0.505, radius=0.505, shapeName=dyn_torsoFlap_Base_r_capsule, rotation=<rot 0 -90 0>}]
 dyn_torsoFlap_back [{shapeType=Box, extentX=2.4427, extentY=11.146, extentZ=23.806, center=<4.4329 -0.1126 -0.0344>, rotation=<rot -4.6545 -90 3.6414>}]
 dyn_torsoFlap_back_Base [{shapeType=Capsule, length=0.505, radius=0.505, shapeName=dyn_torsoFlap_back_Base_capsule, rotation=<rot 0 -90 0>}]
 dyn_torsoFlap_front [{shapeType=Box, extentX=2.6931, extentY=10.4602, extentZ=25.0709, center=<4.1115 0.3209 0.2069>, rotation=<rot 3.1567 -90 4.7853>}]
 dyn_torsoFlap_front_Base [{shapeType=Capsule, length=0.505, radius=0.505, shapeName=dyn_torsoFlap_front_Base_capsule, rotation=<rot 0 -90 0>}]
 dyn_torsoFlap_l [{shapeType=Box, extentX=1.2313, extentY=9.8414, extentZ=23.8734, center=<3.5727 0.0155 0.882>, rotation=<rot -1.72 -90 1.1906>}]
 dyn_torsoFlap_r [{shapeType=Box, extentX=1.4007, extentY=11.7982, extentZ=24.5485, center=<4.7297 0.1461 -0.2417>, rotation=<rot 0.9944 -90 6.8721>}]
 pelvis [{shapeType=Capsule, length=11.0118, radius=8.9855, center=<-0.4785 -1.9086 0>}]
 spine_01 [{shapeType=Capsule, length=9.4951, radius=7.2511, center=<0.9069 0 0>}]
 spine_02 [{shapeType=Capsule, length=4.0066, radius=7.06, center=<0 1.6752 0>}]
 spine_03 [{shapeType=Capsule, length=7.3441, radius=8.8522, center=<0 1.5949 0>, rotation=<rot 90 0 0>}]
}
constraints [
 {
  bone1Name=dyn_torsoFlap_l
  bone2Name=dyn_torsoFlap_Base_l
  swing1LimitDegrees=45
  swing2LimitDegrees=2
  swing2Motion=Locked
  twistLimitDegrees=45
  twistMotion=Locked
 }
 {
  bone1Name=dyn_torsoFlap_back
  bone2Name=dyn_torsoFlap_back_Base
  swing1LimitDegrees=30
  swing2LimitDegrees=45
  swing2Motion=Locked
  twistLimitDegrees=45
  twistMotion=Locked
 }
 {
  bone1Name=dyn_torsoFlap_r
  bone2Name=dyn_torsoFlap_Base_r
  swing1LimitDegrees=45
  swing2LimitDegrees=45
  swing2Motion=Locked
  twistLimitDegrees=45
  twistMotion=Locked
 }
 {
  bone1Name=dyn_torsoFlap_front
  bone2Name=dyn_torsoFlap_front_Base
  swing1LimitDegrees=45
  swing2LimitDegrees=45
  swing2Motion=Locked
  twistLimitDegrees=45
  twistMotion=Locked
 }
 {
  bone1Name=dyn_torsoBox_l
  bone2Name=dyn_torsoBox_Base_l
  swing1LimitDegrees=45
  swing1Motion=Free
  swing2LimitDegrees=45
  swing2Motion=Free
  twistLimitDegrees=45
  twistMotion=Free
 }
 {
  bone1Name=dyn_torsoBox_r
  bone2Name=dyn_torsoBox_Base_r
  swing1LimitDegrees=45
  swing1Motion=Free
  swing2LimitDegrees=45
  swing2Motion=Free
  twistLimitDegrees=45
  twistMotion=Free
 }
 {
  bone1Name=dyn_torsoFlap_Base_l
  bone2Name=spine_03
  swing1LimitDegrees=45
  swing1Motion=Free
  swing2LimitDegrees=45
  swing2Motion=Free
  twistLimitDegrees=45
  twistMotion=Free
 }
 {
  bone1Name=dyn_torsoFlap_back_Base
  bone2Name=spine_03
  swing1LimitDegrees=45
  swing1Motion=Free
  swing2LimitDegrees=45
  swing2Motion=Free
  twistLimitDegrees=45
  twistMotion=Free
 }
 {
  bone1Name=dyn_torsoFlap_front_Base
  bone2Name=spine_03
  swing1LimitDegrees=45
  swing1Motion=Free
  swing2LimitDegrees=45
  swing2Motion=Free
  twistLimitDegrees=45
  twistMotion=Free
 }
 {
  bone1Name=dyn_torsoBox_Base_l
  bone2Name=spine_03
  swing1LimitDegrees=45
  swing1Motion=Free
  swing2LimitDegrees=45
  swing2Motion=Free
  twistLimitDegrees=45
  twistMotion=Free
 }
 {
  bone1Name=dyn_torsoBox_Base_r
  bone2Name=spine_03
  swing1LimitDegrees=45
  swing1Motion=Free
  swing2LimitDegrees=45
  swing2Motion=Free
  twistLimitDegrees=45
  twistMotion=Free
 }
 {
  bone1Name=spine_01
  bone2Name=pelvis
  swing1LimitDegrees=45
  swing1Motion=Free
  swing2LimitDegrees=45
  swing2Motion=Free
  twistLimitDegrees=45
  twistMotion=Free
 }
 {
  bone1Name=spine_02
  bone2Name=spine_01
  swing1LimitDegrees=45
  swing1Motion=Free
  swing2LimitDegrees=45
  swing2Motion=Free
  twistLimitDegrees=45
  twistMotion=Free
 }
 {
  bone1Name=spine_03
  bone2Name=spine_02
  swing1LimitDegrees=45
  swing1Motion=Free
  swing2LimitDegrees=45
  swing2Motion=Free
  twistLimitDegrees=45
  twistMotion=Free
 }
 {bone1Name=dyn_laceOut02_l, bone2Name=dyn_laceOut01_l, swing1LimitDegrees=45, swing2LimitDegrees=45, twistLimitDegrees=5}
 {bone1Name=dyn_laceIn02_l, bone2Name=dyn_laceIn01_l, swing1LimitDegrees=45, swing2LimitDegrees=45, twistLimitDegrees=5}
 {
  bone1Name=ball_r
  bone2Name=pelvis
  swing1LimitDegrees=45
  swing1Motion=Free
  swing2LimitDegrees=45
  swing2Motion=Free
  twistLimitDegrees=45
  twistMotion=Free
 }
 {bone1Name=dyn_laceIn02_r, bone2Name=dyn_laceIn01_r, swing1LimitDegrees=45, swing2LimitDegrees=45, twistLimitDegrees=5}
 {bone1Name=dyn_laceOut02_r, bone2Name=dyn_laceOut01_r, swing1LimitDegrees=45, swing2LimitDegrees=45, twistLimitDegrees=5}
 {bone1Name=dyn_laceOut01_r, bone2Name=dyn_laceOut01_r_MSTR, swing1LimitDegrees=35, swing2LimitDegrees=35, twistLimitDegrees=5}
 {bone1Name=dyn_laceIn01_r, bone2Name=dyn_laceIn01_r_MSTR, swing1LimitDegrees=35, swing2LimitDegrees=35, twistLimitDegrees=5}
 {
  bone1Name=dyn_laceIn01_r_MSTR
  bone2Name=dyn_shoelace_r_Base
  swing1LimitDegrees=45
  swing1Motion=Free
  swing2LimitDegrees=45
  swing2Motion=Free
  twistLimitDegrees=45
  twistMotion=Free
 }
 {
  bone1Name=dyn_laceOut01_r_MSTR
  bone2Name=dyn_shoelace_r_Base
  swing1LimitDegrees=45
  swing1Motion=Free
  swing2LimitDegrees=45
  swing2Motion=Free
  twistLimitDegrees=45
  twistMotion=Free
 }
 {bone1Name=dyn_laceIn01_l, bone2Name=dyn_laceIn01_I_MSTR, swing1LimitDegrees=35, swing2LimitDegrees=35, twistLimitDegrees=5}
 {bone1Name=dyn_laceOut01_l, bone2Name=dyn_laceOut01_l_MSTR, swing1LimitDegrees=35, swing2LimitDegrees=35, twistLimitDegrees=5}
 {
  bone1Name=dyn_laceOut01_l_MSTR
  bone2Name=dyn_shoelace_l_Base
  swing1LimitDegrees=45
  swing1Motion=Free
  swing2LimitDegrees=45
  swing2Motion=Free
  twistLimitDegrees=45
  twistMotion=Free
 }
 {
  bone1Name=dyn_laceIn01_I_MSTR
  bone2Name=dyn_shoelace_l_Base
  swing1LimitDegrees=45
  swing1Motion=Free
  swing2LimitDegrees=45
  swing2Motion=Free
  twistLimitDegrees=45
  twistMotion=Free
 }
 {
  bone1Name=dyn_torsoFlap_Base_r
  bone2Name=spine_03
  swing1LimitDegrees=45
  swing1Motion=Free
  swing2LimitDegrees=45
  swing2Motion=Free
  twistLimitDegrees=45
  twistMotion=Free
 }
]
