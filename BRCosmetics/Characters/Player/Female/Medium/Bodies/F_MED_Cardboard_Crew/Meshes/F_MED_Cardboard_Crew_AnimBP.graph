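# Post-process style Anim Blueprint for the female medium Cardboard Crew body, layering secondary motion onto the pose copied from its attached parent mesh.
# The copied pose feeds a Control Rig, then RigidBody physics in root-bone space with custom gravity.
# A final constant additive yaw offsets the back torso flap's base bone.
# Each update flags whether a backpack or cape is worn, though the AnimGraph never reads that flag.
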AnimBlueprint /BRCosmetics/Characters/Player/Female/Medium/Bodies/F_MED_Cardboard_Crew/Meshes/F_MED_Cardboard_Crew_AnimBP
target_skeleton=&/F_MED_Cardboard_Crew_Skeleton
graphs [
 AnimGraph {nodes [
   n1 = Control Rig(Source=Copy Pose From Mesh() {bUseAttachedParent=true}) {
    alphaBoolBlend {blendOption=Linear}
    controlRigAssetReference {blueprintRigClass=&/F_MED_CardboardCrew_CtrlRig.F_MED_CardboardCrew_CtrlRig_C}
   }
   n2 = RigidBody(ComponentPose=Local To Component(LocalPose=@n1)) {
    bClampLinearTranslationLimitToRefPose=true
    bOverrideWorldGravity=true
    overridePhysicsAsset=&/F_MED_Cardboard_Crew_Physics
    simulationSpace=BaseBoneSpace
    alphaBoolBlend {blendOption=Linear}
    baseBoneRef {boneName=root}
    componentAppliedLinearAccClamp=<3000 3000 3000>
    componentLinearAccScale=<0.4 4 0.4>
    componentLinearVelScale=<0.25 0.25 0.25>
    overrideWorldGravity {z=-980}
   }
   n3 = "Transform (Modify) Bone - Bone: dyn_torsoFlap_back_Base"(ComponentPose=@n2, Alpha=1, Translation="0,0,0", Rotation="0,55,0", Scale="1,1,1") {
    rotationMode=BMM_Additive
    rotationSpace=BCS_BoneSpace
    alphaBoolBlend {blendOption=Linear}
    boneToModify {boneName=dyn_torsoFlap_back_Base}
    rotation {yaw=55}
   }
   Output Pose(Result=Component To Local(ComponentPose=@n3)) {layerGroup=None}
  ]}
 EventGraph {nodes [
   Event BlueprintUpdateAnimation()
   n1 = CheckForBackbling()
   > Set is Wearing Backpack_or_Cape(is Wearing Backpack_or_Cape:bool=OR Boolean(A=@n1, B=@n1.bIsBackblingACape)) -> (Output_Get:bool)
  ]}
]
variables [is Wearing Backpack_or_Cape {type=boolean}]
