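# Physics asset for the face accessory of the female medium Cardboard Crew character, setting up physics for four cardboard head flaps.
# Each corner flap is a base, 01 and 02 chain off dyn_head, with twist locked and narrow swing limits below the free-rotating base joint.
# Boxes and capsules shape the flaps, while head, neck, upper spine, clavicle and upper-arm bodies, linked by free angular constraints, provide surrounding collision volumes.
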PhysicsAsset /BRCosmetics/Characters/Player/Female/Medium/Bodies/F_MED_Cardboard_Crew/Meshes/Parts/F_MED_CardboardCrew_FaceAcc_Physics
bodies {
 clavicle_l [{shapeType=Box, extentX=23.8226, extentY=10.7175, extentZ=10.5114, center=<4.0373 0.0017 -1.426>, rotation=<rot 0 -90 0>}]
 clavicle_r [{shapeType=Capsule, length=10.1051, radius=5.4927, center=<-4.1834 -1.3392 1.9353>, rotation=<rot -6.1751 -90 -90>}]
 dyn_head [{shapeType=Capsule, length=9.7984, radius=7.8552, shapeName=dyn_head_capsule, center=<2.3791 1.2631 -0.4836>, rotation=<rot 90 179.9999 -180>}]
 dyn_headFlapBL_01 [{shapeType=Box, extentX=0.7349, extentY=5.9586, extentZ=14.8498, center=<2.6848 0.1092 -0.0914>, rotation=<rot -0.7297 -90.0001 -2.4024>}]
 dyn_headFlapBL_01_Base [{shapeType=Box, extentX=1.01, extentY=1.01, extentZ=1.01, rotation=<rot 0 -90 0>}]
 dyn_headFlapBL_02 [{shapeType=Box, extentX=0.8329, extentY=6.2561, extentZ=13.8612, center=<2.7331 0.0626 -0.2327>, rotation=<rot 1.513 -91.4752 7.6421>}]
 dyn_headFlapBR_01 [
  {shapeType=Capsule, length=13.7217, radius=0.9037, center=<4.7663 0.2568 -0.2217>, rotation=<rot -2.6634 0 0>}
  {shapeType=Box, extentX=0.7044, extentY=5.4021, extentZ=14.1417, center=<2.6075 0.0986 -0.4022>, rotation=<rot 3.5385 -90 2.6253>}
 ]
 dyn_headFlapBR_01_Base [{shapeType=Box, extentX=1.01, extentY=1.01, extentZ=1.01, rotation=<rot 0 -90 0>}]
 dyn_headFlapBR_02 [
  {shapeType=Capsule, length=13.4009, radius=1.687, center=<-0.4345 0.3859 -0.5909>, rotation=<rot 8.1165 0 0>}
  {shapeType=Capsule, length=14.2203, radius=0.8677, center=<6.2919 -0.1133 0.2139>, rotation=<rot 10.2701 0 0>}
 ]
 dyn_headFlapFL_01 [{shapeType=Box, extentX=0.7202, extentY=4.695, extentZ=13.8423, center=<1.9303 0.1132 -0.2681>, rotation=<rot 4.4781 -90.0001 -3.7969>}]
 dyn_headFlapFL_01_Base [{shapeType=Box, extentX=1.01, extentY=1.01, extentZ=1.01, rotation=<rot 0 -90 0>}]
 dyn_headFlapFL_02 [{shapeType=Box, extentX=0.8623, extentY=6.7386, extentZ=13.7742, center=<3.1698 -0.1529 0.0532>, rotation=<rot 0.4818 -89.9995 -0.257>}]
 dyn_headFlapFR_01 [{shapeType=Box, extentX=0.7059, extentY=4.747, extentZ=13.3555, center=<1.963 0.1375 -0.2776>, rotation=<rot -2.728 -89.9988 3.0102>}]
 dyn_headFlapFR_01_Base [{shapeType=Box, extentX=1.01, extentY=1.01, extentZ=1.01, rotation=<rot 0 -90 0>}]
 dyn_headFlapFR_02 [
  {shapeType=Capsule, length=13.4143, radius=0.8765, center=<0 -0.2217 0>}
  {shapeType=Capsule, length=12.842, radius=1.4489, center=<5.9543 0.0767 0>}
 ]
 neck_01 [{shapeType=Capsule, length=5.3742, radius=3.5217, rotation=<rot 90 0 0>}]
 neck_02 [{shapeType=Capsule, length=5.3742, radius=3.5217, rotation=<rot 90 0 0>}]
 spine_05 [{shapeType=Sphere, radius=6.3949, center=<7.9604 0.0001 0>}]
 upperarm_l [{shapeType=Capsule, length=8.1685, radius=4.7551, center=<5 0 0>, rotation=<rot 90 0 0>}]
 upperarm_r [{shapeType=Capsule, length=8.1685, radius=4.7551, center=<-5 0 0>, rotation=<rot 90 0 0>}]
}
constraints [
 {
  bone1Name=dyn_headFlapBR_01_Base
  bone2Name=dyn_head
  swing1LimitDegrees=45
  swing1Motion=Free
  swing2LimitDegrees=45
  swing2Motion=Free
  twistLimitDegrees=45
  twistMotion=Free
 }
 {
  bone1Name=dyn_headFlapBR_01
  bone2Name=dyn_headFlapBR_01_Base
  swing1LimitDegrees=20
  swing2LimitDegrees=12
  twistLimitDegrees=45
  twistMotion=Locked
 }
 {
  bone1Name=dyn_headFlapBR_02
  bone2Name=dyn_headFlapBR_01
  swing1LimitDegrees=20
  swing2LimitDegrees=8
  twistLimitDegrees=45
  twistMotion=Locked
 }
 {
  bone1Name=dyn_headFlapBL_01_Base
  bone2Name=dyn_head
  swing1LimitDegrees=45
  swing1Motion=Free
  swing2LimitDegrees=45
  swing2Motion=Free
  twistLimitDegrees=45
  twistMotion=Free
 }
 {
  bone1Name=dyn_headFlapBL_01
  bone2Name=dyn_headFlapBL_01_Base
  swing1LimitDegrees=20
  swing2LimitDegrees=12
  twistLimitDegrees=45
  twistMotion=Locked
 }
 {
  bone1Name=dyn_headFlapBL_02
  bone2Name=dyn_headFlapBL_01
  swing1LimitDegrees=20
  swing2LimitDegrees=8
  twistLimitDegrees=45
  twistMotion=Locked
 }
 {
  bone1Name=dyn_headFlapFL_01_Base
  bone2Name=dyn_head
  swing1LimitDegrees=45
  swing1Motion=Free
  swing2LimitDegrees=45
  swing2Motion=Free
  twistLimitDegrees=45
  twistMotion=Free
 }
 {
  bone1Name=dyn_headFlapFL_01
  bone2Name=dyn_headFlapFL_01_Base
  swing1LimitDegrees=20
  swing2LimitDegrees=12
  twistLimitDegrees=45
  twistMotion=Locked
 }
 {
  bone1Name=dyn_headFlapFL_02
  bone2Name=dyn_headFlapFL_01
  swing1LimitDegrees=20
  swing2LimitDegrees=8
  twistLimitDegrees=45
  twistMotion=Locked
 }
 {
  bone1Name=dyn_headFlapFR_01_Base
  bone2Name=dyn_head
  swing1LimitDegrees=45
  swing1Motion=Free
  swing2LimitDegrees=45
  swing2Motion=Free
  twistLimitDegrees=45
  twistMotion=Free
 }
 {
  bone1Name=dyn_headFlapFR_01
  bone2Name=dyn_headFlapFR_01_Base
  swing1LimitDegrees=20
  swing2LimitDegrees=12
  twistLimitDegrees=45
  twistMotion=Locked
 }
 {
  bone1Name=dyn_headFlapFR_02
  bone2Name=dyn_headFlapFR_01
  swing1LimitDegrees=20
  swing2LimitDegrees=8
  twistLimitDegrees=45
  twistMotion=Locked
 }
 {
  bone1Name=upperarm_l
  bone2Name=clavicle_l
  swing1LimitDegrees=45
  swing1Motion=Free
  swing2LimitDegrees=45
  swing2Motion=Free
  twistLimitDegrees=45
  twistMotion=Free
 }
 {
  bone1Name=upperarm_r
  bone2Name=clavicle_r
  swing1LimitDegrees=45
  swing1Motion=Free
  swing2LimitDegrees=45
  swing2Motion=Free
  twistLimitDegrees=45
  twistMotion=Free
 }
 {
  bone1Name=clavicle_l
  bone2Name=spine_05
  swing1LimitDegrees=45
  swing1Motion=Free
  swing2LimitDegrees=45
  swing2Motion=Free
  twistLimitDegrees=45
  twistMotion=Free
 }
 {
  bone1Name=clavicle_r
  bone2Name=spine_05
  swing1LimitDegrees=45
  swing1Motion=Free
  swing2LimitDegrees=45
  swing2Motion=Free
  twistLimitDegrees=45
  twistMotion=Free
 }
 {
  bone1Name=dyn_head
  bone2Name=neck_02
  swing1LimitDegrees=45
  swing1Motion=Free
  swing2LimitDegrees=45
  swing2Motion=Free
  twistLimitDegrees=45
  twistMotion=Free
 }
 {
  bone1Name=neck_01
  bone2Name=spine_05
  swing1LimitDegrees=45
  swing1Motion=Free
  swing2LimitDegrees=45
  swing2Motion=Free
  twistLimitDegrees=45
  twistMotion=Free
 }
 {
  bone1Name=neck_02
  bone2Name=neck_01
  swing1LimitDegrees=45
  swing1Motion=Free
  swing2LimitDegrees=45
  swing2Motion=Free
  twistLimitDegrees=45
  twistMotion=Free
 }
]
