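# Animation Blueprint for the face accessory part of the female Cardboard Crew cosmetic, following the body it is attached to.
# It copies the pose from the attached parent mesh, then runs a Control Rig fed targeting, crouch-targeting, parachute and skydive-floating flags.
# A RigidBody node then simulates secondary physics relative to spine_05, using an override physics asset and damped component motion.
# The event graph updates the crouch-targeting flag each frame from the targeting and crouching states.
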AnimBlueprint /BRCosmetics/Characters/Player/Female/Medium/Bodies/F_MED_Cardboard_Crew/Meshes/Parts/F_MED_CardboardCrew_FaceAcc_AnimBP
target_skeleton=&/F_MED_CardboardCrew_FaceAcc_Skeleton
graphs [
 AnimGraph {nodes [
   n1 = Control Rig!(
    Source=Copy Pose From Mesh() {bUseAttachedParent=true}
    Alpha
    is_Targeting:bool=Get bIsTargeting()
    isParashuteOpened:bool=Get bIsParachuteOpened()
    TargetingOnCrouch:bool=Get TargetingCroching() -> (TargetingCroching:bool)
    isSkyFloating:bool=Get bIsSkydiveFloating()
   ) -> (Pose) {
    alphaBoolBlend {blendOption=Linear}
    controlRigAssetReference {blueprintRigClass=&/F_MED_CardboardCrew_FaceAcc_CtrlRig.F_MED_CardboardCrew_FaceAcc_CtrlRig_C}
   }
   n2 = RigidBody(ComponentPose=Local To Component(LocalPose=@n1)) {
    bClampLinearTranslationLimitToRefPose=true
    bOverrideWorldGravity=true
    overridePhysicsAsset=&/F_MED_CardboardCrew_FaceAcc_Physics
    simulationSpace=BaseBoneSpace
    alphaBoolBlend {blendOption=Linear}
    baseBoneRef {boneName=spine_05}
    componentAppliedLinearAccClamp=<3000 3000 3000>
    componentLinearAccScale=<0.5 0.5 0.5>
    componentLinearVelScale=<0.2 0.2 0.2>
    overrideWorldGravity {z=-980}
   }
   Output Pose(Result=Component To Local(ComponentPose=@n2)) {layerGroup=None}
  ]}
 EventGraph {nodes [
   Event BlueprintUpdateAnimation()
   > Set TargetingCroching(TargetingCroching:bool=AND Boolean(A=Get bIsTargeting(), B=Get bIsCrouching())) -> (Output_Get:bool)
   TryGetPawnOwner()
  ]}
]
variables [
 TargetingCroching {type=boolean}
 TargetingCrochSprinting {type=boolean}
]
